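# Looping sound cue for the gameplay area barrier in LimeRock, routed through the Fort_Events sound class.
# A single looping wave player feeds the output at reduced volume, with spatial falloff from a gameplay barrier attenuation asset.
# Its own concurrency override caps simultaneous instances at a few voices, with volume scaling and a timed release for stolen voices.
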SoundCue /LimeRock/Assets/Audio/SFX/Misc/LimeRock_GameplayArea_Barrier_LP
root=WavePlayer_1
nodes [{id=WavePlayer_1, properties {bLooping=true, soundWaveAssetPtr=/LimeRock/Assets/Audio/SFX/Misc/Wav/LimeRock_GameplayArea_Barrier_LP}}]
properties {
 attenuationSettings=/LimeRock/Assets/Audio/Settings/ATT_LimeRock_GameplayBarriers
 bOverrideConcurrency=true
 maxDistance=1273.5632
 soundClassObject=&Fort_Events
 volumeMultiplier=0.6
 concurrencyOverrides {maxCount=3, voiceStealReleaseTime=0.5, volumeScale=0.75, volumeScaleAttackTime=0.25}
}
