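# Looping SoundCue for the outer barrier of the LimeRock gameplay area.
# A single looping wave player reuses the Bubble Shield "Silver Blazer" loop at reduced volume, spatialized with LimeRock's small attenuation settings in the Fort_Events sound class.
# Overridden concurrency caps simultaneous voices and scales their volume with short attack and release fades.
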SoundCue /LimeRock/Assets/Audio/SFX/Misc/LimeRock_GameplayArea_OuterBarrier_LP
root=WavePlayer_1
nodes [{id=WavePlayer_1, properties {bLooping=true, soundWaveAssetPtr=/Game/Athena/Sounds/Items/Bubble_Shield/Silver_Blazer_Loop_01}}]
properties {
 attenuationSettings=/LimeRock/Assets/Audio/Settings/ATT_LimeRock_Small
 bOverrideConcurrency=true
 maxDistance=2683.9082
 soundClassObject=&Fort_Events
 volumeMultiplier=0.6
 concurrencyOverrides {maxCount=3, voiceStealReleaseTime=0.5, volumeScale=0.75, volumeScaleAttackTime=0.25}
}
