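# LimeRock_HoverAura_End is a SoundCue in LimeRock's misc SFX, named as the closing sound of a hover aura effect.
# A single wave player reuses the HoverAura_Start wave, raised in pitch and slightly lowered in volume, with no separate end recording.
# It plays through the large POI attenuation and the Fort_Events sound class.
# A concurrency override caps simultaneous instances, ducking their volume and fading out stolen voices.
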SoundCue /LimeRock/Assets/Audio/SFX/Misc/LimeRock_HoverAura_End
root=WavePlayer_3
nodes [{id=WavePlayer_3, properties {soundWaveAssetPtr=/LimeRock/Assets/Audio/SFX/Misc/Wav/LimeRock_HoverAura_Start}}]
properties {
 attenuationSettings=/LimeRock/Assets/Audio/Settings/ATT_LimeRock_POI_Large
 bOverrideConcurrency=true
 maxDistance=40000
 pitchMultiplier=1.6
 soundClassObject=&Fort_Events
 volumeMultiplier=0.8
 concurrencyOverrides {maxCount=3, voiceStealReleaseTime=0.5, volumeScale=0.75, volumeScaleAttackTime=0.25}
}
