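# LimeRock hover aura sound cue, pairing a start one-shot with a sustained loop in the Fort_Events sound class.
# A Mixer blends the two briefly delayed layers, weighting the loop above the start sound.
# An Enveloper pitches the loop down and holds it silent briefly, then fades it in and settles slightly lower.
# Large POI attenuation and a concurrency override with volume scaling limit simultaneous instances.
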SoundCue /LimeRock/Assets/Audio/SFX/Misc/LimeRock_HoverAura_LP
root=Mixer_0
nodes [
 {id=Delay_0, children [WavePlayer_3], properties {delayMax=0.25, delayMin=0.25}}
 {id=Delay_1, children [WavePlayer_0], properties {delayMax=0.25, delayMin=0.25}}
 {
  id=Enveloper_0
  children [Delay_1]
  properties {
   pitchMax=0.7
   pitchMin=0.7
   pitchCurve {editorCurveData {keys [{time=0}]}}
   volumeCurve {editorCurveData {keys [<at -0.0038 0>,<at 1.3 0>,<at 3 1>,<at 8 0.7>]}}
  }
 }
 {id=Mixer_0, children [Enveloper_0,Delay_0], properties {inputVolume [1.2,0.6]}}
 {id=WavePlayer_0, properties {bLooping=true, soundWaveAssetPtr=/LimeRock/Assets/Audio/SFX/Misc/Wav/LimeRock_HoverAura_LP}}
 {id=WavePlayer_3, properties {soundWaveAssetPtr=/LimeRock/Assets/Audio/SFX/Misc/Wav/LimeRock_HoverAura_Start}}
]
properties {
 attenuationSettings=/LimeRock/Assets/Audio/Settings/ATT_LimeRock_POI_Large
 bOverrideConcurrency=true
 maxDistance=40000
 soundClassObject=&Fort_Events
 volumeMultiplier=1
 concurrencyOverrides {maxCount=3, voiceStealReleaseTime=0.5, volumeScale=0.75, volumeScaleAttackTime=0.25}
}
