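# Looping night-time farm ambience Sound Cue for LimeRock, routed to the Fort_Ambient_Athena sound class with matching attenuation.
# A Quality Level node at the root selects between two looping wave players.
# The first two quality slots share the quad recording and the last two share the stereo one, so channel count varies with audio quality.
# The cue plays at reduced volume.
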SoundCue /LimeRock/Assets/Audio/SFX/Misc/SC_LimeRock_Ambience_Night_LP
root=QualityLevel_0
nodes [
 {id=QualityLevel_0, children [WavePlayer_4,WavePlayer_4,WavePlayer_5,WavePlayer_5]}
 {id=WavePlayer_4, properties {bLooping=true, soundWaveAssetPtr=/LimeRock/Assets/Audio/SFX/Misc/Wav/Ambience_Farm_Loop_Night_Quad}}
 {id=WavePlayer_5, properties {bLooping=true, soundWaveAssetPtr=/LimeRock/Assets/Audio/SFX/Misc/Wav/Ambience_Farm_Loop_Night_Stereo}}
]
properties {
 attenuationSettings=&Fort_Ambient_Athena_Att
 maxDistance=2097152
 soundClassObject=&Fort_Ambient_Athena
 volumeMultiplier=0.3
}
